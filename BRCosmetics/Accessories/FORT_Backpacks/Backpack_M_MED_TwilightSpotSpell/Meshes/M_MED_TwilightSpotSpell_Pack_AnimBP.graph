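# Post-process style AnimBP for the TwilightSpotSpell backpack, simulating secondary motion on its leg, arm and towel bones with AnimDynamics.
# The cached input pose feeds two chains selected by bIsInFrontEnd; the front-end chain damps horizontal motion influence.
# In-game, legs always simulate, while skydive floating swaps arm and towel dynamics and adds towel rotation offsets via timed alpha blends.
# Several other AnimDynamics chains are unconnected to the output.
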AnimBlueprint /BRCosmetics/Accessories/FORT_Backpacks/Backpack_M_MED_TwilightSpotSpell/Meshes/M_MED_TwilightSpotSpell_Pack_AnimBP
target_skeleton=&/M_MED_TwilightSpotSpell_Pack_Skeleton
graphs [
 AnimGraph {nodes [
   n1 = Anim Dynamics - Bone: leg_l(Alpha=1, ExternalForce="0,0,0") {
    angularDampingOverride=1
    angularSpringConstant=500
    bOverrideAngularDamping=true
    bUseGravityOverride=true
    simulationSpace=RootRelative
    alphaBoolBlend {blendOption=Linear}
    boundBone {boneName=leg_l}
    componentAppliedLinearAccClamp=<100 100 100>
    componentLinearAccScale=<0.6 0.6 0.6>
    componentLinearVelScale=<0.7 0.7 0.7>
    gravityOverride {z=-980}
    physicsBodyDefinitions [{
      boundBone {boneName=leg_l}
      boxExtents=<5 5 5>
      constraintSetup {angularLimitsMax=<35 15 15>, angularLimitsMin=<-35 -45 -15>, angularTarget=<0 0 0>}
      localJointOffset=<0.2 0 0>
     }]
   }
   Anim Dynamics - Bone: leg_r(ComponentPose=@n1, Alpha=1, ExternalForce="0,0,0") {
    angularDampingOverride=1
    angularSpringConstant=500
    bOverrideAngularDamping=true
    bUseGravityOverride=true
    simulationSpace=RootRelative
    alphaBoolBlend {blendOption=Linear}
    boundBone {boneName=leg_r}
    componentAppliedLinearAccClamp=<300 300 300>
    componentLinearAccScale=<0.6 0.6 0.6>
    componentLinearVelScale=<0.7 0.7 0.7>
    gravityOverride {z=-850}
    physicsBodyDefinitions [{
      boundBone {boneName=leg_r}
      boxExtents=<7 7 7>
      constraintSetup {angularLimitsMax=<45 15 15>, angularLimitsMin=<-45 -55 -15>, angularTarget=<0 0 0>}
      localJointOffset=<0.5 0 0>
     }]
   }
   n2 = Anim Dynamics - Bone: arm_l(Alpha=1, ExternalForce="0,0,0") {
    angularDampingOverride=0.999
    angularSpringConstant=200
    bOverrideAngularDamping=true
    bUseGravityOverride=true
    simulationSpace=RootRelative
    alphaBoolBlend {blendOption=Linear}
    boundBone {boneName=arm_l}
    componentAppliedLinearAccClamp=<200 200 700>
    componentLinearAccScale=<0.6 0.6 0.6>
    componentLinearVelScale=<0.7 0.7 0.7>
    gravityOverride {z=-500}
    physicsBodyDefinitions [{
      boundBone {boneName=arm_l}
      boxExtents=<9 11 8>
      constraintSetup {angularLimitsMax=<5 15 5>, angularLimitsMin=<-5 -15 -25>, angularTarget=<0 0 0>}
      localJointOffset=<1 0 0>
     }]
   }
   Anim Dynamics - Bone: arm_r(ComponentPose=@n2, Alpha=1, ExternalForce="0,0,0") {
    angularDampingOverride=0.999
    angularSpringConstant=200
    bOverrideAngularDamping=true
    bUseGravityOverride=true
    simulationSpace=RootRelative
    alphaBoolBlend {blendOption=Linear}
    boundBone {boneName=arm_r}
    componentAppliedLinearAccClamp=<200 200 700>
    componentLinearAccScale=<0.6 0.6 0.6>
    componentLinearVelScale=<0.7 0.7 0.7>
    gravityOverride {z=-500}
    physicsBodyDefinitions [{
      boundBone {boneName=arm_r}
      boxExtents=<9 11 8>
      constraintSetup {angularLimitsMax=<5 15 25>, angularLimitsMin=<-5 -15 -5>, angularTarget=<0 0 0>}
      localJointOffset=<1 0 0>
     }]
   }
   n3 = Anim Dynamics - Bone: leg_l(Alpha=1, ExternalForce="0,0,0") {
    angularDampingOverride=0.9989
    angularSpringConstant=9000
    bAngularSpring=true
    bOverrideAngularDamping=true
    bUseGravityOverride=true
    simulationSpace=RootRelative
    alphaBoolBlend {blendOption=Linear}
    boundBone {boneName=leg_l}
    componentAppliedLinearAccClamp=<500 500 500>
    componentLinearAccScale=<1.5 1.5 1.5>
    componentLinearVelScale=<1.5 1.5 1.5>
    gravityOverride {z=-980}
    physicsBodyDefinitions [{
      boundBone {boneName=leg_l}
      boxExtents=<4 8 8>
      constraintSetup {angularLimitsMax=<5 20 2>, angularLimitsMin=<-5 -25 -2>, angularTarget=<0 0 0>}
      localJointOffset=<1 0 0>
     }]
   }
   n4 = Anim Dynamics - Bone: leg_r(ComponentPose=@n3, Alpha=1, ExternalForce="0,0,0") {
    angularDampingOverride=0.9989
    angularSpringConstant=1000
    bAngularSpring=true
    bOverrideAngularDamping=true
    bUseGravityOverride=true
    simulationSpace=RootRelative
    alphaBoolBlend {blendOption=Linear}
    boundBone {boneName=leg_r}
    componentAppliedLinearAccClamp=<1000 1000 1000>
    componentLinearAccScale=<1.2 1.2 0.8>
    componentLinearVelScale=<1.2 1.2 0.9>
    gravityOverride {z=-850}
    physicsBodyDefinitions [{
      boundBone {boneName=leg_r}
      boxExtents=<5 7.5 7.5>
      constraintSetup {angularLimitsMax=<10 25 2>, angularLimitsMin=<-10 -25 -2>, angularTarget=<0 0 0>}
      localJointOffset=<0.9 0 0>
     }]
   }
   n5 = Anim Dynamics - Bone: arm_l(ComponentPose=@n4, Alpha=1, ExternalForce="0,0,0") {
    angularDampingOverride=0.9989
    angularSpringConstant=7000
    bAngularSpring=true
    bOverrideAngularDamping=true
    bUseGravityOverride=true
    simulationSpace=RootRelative
    alphaBoolBlend {blendOption=Linear}
    boundBone {boneName=arm_l}
    componentAppliedLinearAccClamp=<200 200 200>
    componentLinearAccScale=<0.8 0.8 0.8>
    componentLinearVelScale=<0.8 0.8 0.8>
    gravityOverride {z=-500}
    physicsBodyDefinitions [{
      boundBone {boneName=arm_l}
      boxExtents=<7 7 7>
      constraintSetup {angularLimitsMax=<5 30 5>, angularLimitsMin=<-5 -25 -50>, angularTarget=<0 0 0>}
      localJointOffset=<1.7 0 0>
     }]
   }
   n6 = Anim Dynamics - Bone: arm_r(ComponentPose=@n5, Alpha=1, ExternalForce="0,0,0") {
    angularDampingOverride=0.9989
    angularSpringConstant=5000
    bAngularSpring=true
    bOverrideAngularDamping=true
    bUseGravityOverride=true
    simulationSpace=RootRelative
    alphaBoolBlend {blendOption=Linear}
    boundBone {boneName=arm_r}
    componentAppliedLinearAccClamp=<190 190 690>
    componentLinearAccScale=<0.65 0.65 0.65>
    componentLinearVelScale=<0.75 0.75 0.75>
    gravityOverride {z=-500}
    physicsBodyDefinitions [{
      boundBone {boneName=arm_r}
      boxExtents=<4 5 6>
      constraintSetup {angularLimitsMax=<5 35 50>, angularLimitsMin=<-5 -30 -5>, angularTarget=<0 0 0>}
      localJointOffset=<1 0 0>
     }]
   }
   n7 = Anim Dynamics - Bone: towel_l(ComponentPose=@n6, Alpha=1, ExternalForce="0,0,0") {
    angularDampingOverride=0.9989
    angularSpringConstant=1500
    bAngularSpring=true
    bOverrideAngularDamping=true
    bUseGravityOverride=true
    simulationSpace=RootRelative
    alphaBoolBlend {blendOption=Linear}
    boundBone {boneName=towel_l}
    componentAppliedLinearAccClamp=<200 200 700>
    componentLinearAccScale=<0.6 0.6 0.6>
    componentLinearVelScale=<0.7 0.7 0.7>
    gravityOverride {z=-500}
    physicsBodyDefinitions [{
      boundBone {boneName=towel_l}
      boxExtents=<4 5.5 5.5>
      constraintSetup {angularLimitsMax=<15 5 5>, angularLimitsMin=<-20 -30 -25>, angularTarget=<0 0 0>}
      localJointOffset=<1 0 0>
     }]
   }
   Anim Dynamics - Bone: towel_r(ComponentPose=@n7, Alpha=1, ExternalForce="0,0,0") {
    angularDampingOverride=0.9989
    angularSpringConstant=6000
    bAngularSpring=true
    bOverrideAngularDamping=true
    bUseGravityOverride=true
    simulationSpace=RootRelative
    alphaBoolBlend {blendOption=Linear}
    boundBone {boneName=towel_r}
    componentAppliedLinearAccClamp=<350 350 350>
    componentLinearAccScale=<0.6 0.6 0.6>
    componentLinearVelScale=<0.7 0.7 0.7>
    gravityOverride {z=-500}
    physicsBodyDefinitions [{
      boundBone {boneName=towel_r}
      boxExtents=<5.5 5.5 5.5>
      constraintSetup {angularLimitsMax=<30 5 25>, angularLimitsMin=<-5 -30 -5>, angularTarget=<0 0 0>}
      localJointOffset=<0.7 0 0>
     }]
   }
   n8 = Anim Dynamics - Bone: leg_l(Alpha=1, ExternalForce="0,0,0") {
    angularDampingOverride=1
    angularSpringConstant=8000
    bAngularSpring=true
    bOverrideAngularDamping=true
    bUseGravityOverride=true
    simulationSpace=RootRelative
    alphaBoolBlend {blendOption=Linear}
    boundBone {boneName=leg_l}
    componentAppliedLinearAccClamp=<1200 1200 500>
    componentLinearAccScale=<1.2 1.2 0.8>
    componentLinearVelScale=<1.2 1.2 0.9>
    gravityOverride {z=-980}
    physicsBodyDefinitions [{
      boundBone {boneName=leg_l}
      boxExtents=<5 10 10>
      constraintSetup {angularLimitsMax=<10 20 2>, angularLimitsMin=<-10 -10 -2>, angularTarget=<0 0 0>}
      localJointOffset=<0.5 0 0>
     }]
   }
   Anim Dynamics - Bone: leg_r(ComponentPose=@n8, Alpha=1, ExternalForce="0,0,0") {
    angularDampingOverride=1
    angularSpringConstant=10000
    bAngularSpring=true
    bOverrideAngularDamping=true
    bUseGravityOverride=true
    simulationSpace=RootRelative
    alphaBoolBlend {blendOption=Linear}
    boundBone {boneName=leg_r}
    componentAppliedLinearAccClamp=<1200 1200 500>
    componentLinearAccScale=<1.2 1.2 0.8>
    componentLinearVelScale=<1.2 1.2 0.9>
    gravityOverride {z=-850}
    physicsBodyDefinitions [{
      boundBone {boneName=leg_r}
      boxExtents=<5 10 10>
      constraintSetup {angularLimitsMax=<10 20 2>, angularLimitsMin=<-10 -10 -2>, angularTarget=<0 0 0>}
      localJointOffset=<0.5 0 0>
     }]
   }
   n9 = Anim Dynamics - Bone: leg_l(ComponentPose=Local To Component(LocalPose=Use cached pose 'Base_pose'()), Alpha=1, ExternalForce="0,0,0") {
    angularDampingOverride=0.9989
    angularSpringConstant=9000
    bAngularSpring=true
    bOverrideAngularDamping=true
    bUseGravityOverride=true
    simulationSpace=RootRelative
    alphaBoolBlend {blendOption=Linear}
    boundBone {boneName=leg_l}
    componentAppliedLinearAccClamp=<500 500 500>
    componentLinearAccScale=<0.1 0.1 1.5>
    componentLinearVelScale=<0.2 0.2 1.5>
    gravityOverride {z=-980}
    physicsBodyDefinitions [{
      boundBone {boneName=leg_l}
      boxExtents=<4 8 8>
      constraintSetup {angularLimitsMax=<5 20 2>, angularLimitsMin=<-5 -25 -2>, angularTarget=<0 0 0>}
      localJointOffset=<1 0 0>
     }]
   }
   n10 = Anim Dynamics - Bone: leg_r(ComponentPose=@n9, Alpha=1, ExternalForce="0,0,0") {
    angularDampingOverride=0.9989
    angularSpringConstant=1000
    bAngularSpring=true
    bOverrideAngularDamping=true
    bUseGravityOverride=true
    simulationSpace=RootRelative
    alphaBoolBlend {blendOption=Linear}
    boundBone {boneName=leg_r}
    componentAppliedLinearAccClamp=<1000 1000 1000>
    componentLinearAccScale=<0.1 0.1 0.8>
    componentLinearVelScale=<0.3 0.3 0.9>
    gravityOverride {z=-850}
    physicsBodyDefinitions [{
      boundBone {boneName=leg_r}
      boxExtents=<5 7.5 7.5>
      constraintSetup {angularLimitsMax=<10 25 2>, angularLimitsMin=<-10 -25 -2>, angularTarget=<0 0 0>}
      localJointOffset=<0.9 0 0>
     }]
   }
   n11 = Anim Dynamics - Bone: arm_l(ComponentPose=@n10, Alpha=1, ExternalForce="0,0,0") {
    angularDampingOverride=0.9989
    angularSpringConstant=7000
    bAngularSpring=true
    bOverrideAngularDamping=true
    bUseGravityOverride=true
    simulationSpace=RootRelative
    alphaBoolBlend {blendOption=Linear}
    boundBone {boneName=arm_l}
    componentAppliedLinearAccClamp=<200 200 200>
    componentLinearAccScale=<0.1 0.1 0.8>
    componentLinearVelScale=<0.2 0.2 0.8>
    gravityOverride {z=-500}
    physicsBodyDefinitions [{
      boundBone {boneName=arm_l}
      boxExtents=<7 7 7>
      constraintSetup {angularLimitsMax=<5 30 5>, angularLimitsMin=<-5 -25 -50>, angularTarget=<0 0 0>}
      localJointOffset=<1.7 0 0>
     }]
   }
   n12 = Anim Dynamics - Bone: arm_r(ComponentPose=@n11, Alpha=1, ExternalForce="0,0,0") {
    angularDampingOverride=0.9989
    angularSpringConstant=5000
    bAngularSpring=true
    bOverrideAngularDamping=true
    bUseGravityOverride=true
    simulationSpace=RootRelative
    alphaBoolBlend {blendOption=Linear}
    boundBone {boneName=arm_r}
    componentAppliedLinearAccClamp=<190 190 690>
    componentLinearAccScale=<0.1 0.1 0.65>
    componentLinearVelScale=<0.3 0.3 0.75>
    gravityOverride {z=-500}
    physicsBodyDefinitions [{
      boundBone {boneName=arm_r}
      boxExtents=<4 5 6>
      constraintSetup {angularLimitsMax=<5 35 50>, angularLimitsMin=<-5 -30 -5>, angularTarget=<0 0 0>}
      localJointOffset=<1 0 0>
     }]
   }
   n13 = Anim Dynamics - Bone: towel_l(ComponentPose=@n12, Alpha=1, ExternalForce="0,0,0") {
    angularDampingOverride=0.9989
    angularSpringConstant=1500
    bAngularSpring=true
    bOverrideAngularDamping=true
    bUseGravityOverride=true
    simulationSpace=RootRelative
    alphaBoolBlend {blendOption=Linear}
    boundBone {boneName=towel_l}
    componentAppliedLinearAccClamp=<200 200 700>
    componentLinearAccScale=<0.1 0.1 0.6>
    componentLinearVelScale=<0.3 0.3 0.7>
    gravityOverride {z=-500}
    physicsBodyDefinitions [{
      boundBone {boneName=towel_l}
      boxExtents=<4 5.5 5.5>
      constraintSetup {angularLimitsMax=<15 5 5>, angularLimitsMin=<-20 -30 -25>, angularTarget=<0 0 0>}
      localJointOffset=<1 0 0>
     }]
   }
   n14 = Anim Dynamics - Bone: towel_r(ComponentPose=@n13, Alpha=1, ExternalForce="0,0,0") {
    angularDampingOverride=0.9989
    angularSpringConstant=6000
    bAngularSpring=true
    bOverrideAngularDamping=true
    bUseGravityOverride=true
    simulationSpace=RootRelative
    alphaBoolBlend {blendOption=Linear}
    boundBone {boneName=towel_r}
    componentAppliedLinearAccClamp=<350 350 350>
    componentLinearAccScale=<0.1 0.1 0.6>
    componentLinearVelScale=<0.2 0.2 0.7>
    gravityOverride {z=-500}
    physicsBodyDefinitions [{
      boundBone {boneName=towel_r}
      boxExtents=<5.5 5.5 5.5>
      constraintSetup {angularLimitsMax=<30 5 25>, angularLimitsMin=<-5 -30 -5>, angularTarget=<0 0 0>}
      localJointOffset=<0.7 0 0>
     }]
   }
   n15 = Anim Dynamics - Bone: leg_l(ComponentPose=Local To Component(LocalPose=Use cached pose 'Base_pose'()), Alpha=1, ExternalForce="0,0,0") {
    angularDampingOverride=1
    angularSpringConstant=8000
    bAngularSpring=true
    bOverrideAngularDamping=true
    bUseGravityOverride=true
    simulationSpace=RootRelative
    alphaBoolBlend {blendOption=Linear}
    boundBone {boneName=leg_l}
    componentAppliedLinearAccClamp=<500 500 500>
    componentLinearAccScale=<1.2 1.2 0.8>
    componentLinearVelScale=<1.2 1.2 0.9>
    gravityOverride {z=-980}
    physicsBodyDefinitions [{
      boundBone {boneName=leg_l}
      boxExtents=<5 10 10>
      constraintSetup {angularLimitsMax=<10 20 2>, angularLimitsMin=<-10 -25 -2>, angularTarget=<0 0 0>}
      localJointOffset=<0.5 0 0>
     }]
   }
   n16 = Anim Dynamics - Bone: leg_r(ComponentPose=@n15, Alpha=1, ExternalForce="0,0,0") {
    angularDampingOverride=1
    angularSpringConstant=10000
    bAngularSpring=true
    bOverrideAngularDamping=true
    bUseGravityOverride=true
    simulationSpace=RootRelative
    alphaBoolBlend {blendOption=Linear}
    boundBone {boneName=leg_r}
    componentAppliedLinearAccClamp=<700 700 700>
    componentLinearAccScale=<1.2 1.2 0.8>
    componentLinearVelScale=<1.2 1.2 0.9>
    gravityOverride {z=-850}
    physicsBodyDefinitions [{
      boundBone {boneName=leg_r}
      boxExtents=<5 10 10>
      constraintSetup {angularLimitsMax=<10 20 2>, angularLimitsMin=<-10 -25 -2>, angularTarget=<0 0 0>}
      localJointOffset=<0.5 0 0>
     }]
   }
   n17 = Get NotInSkydive() -> (NotInSkydive:bool)
   n18 = Anim Dynamics - Bone: towel_l!(ComponentPose=@n16, bAlphaBoolEnabled:bool=@n17, ExternalForce="0,0,0") -> (Pose) {
    alphaInputType=Bool
    angularDampingOverride=0.999
    angularSpringConstant=200
    bOverrideAngularDamping=true
    bUseGravityOverride=true
    simulationSpace=RootRelative
    alphaBoolBlend {blendInTime=0.1, blendOption=Linear, blendOutTime=0.1}
    boundBone {boneName=towel_l}
    componentAppliedLinearAccClamp=<200 200 700>
    componentLinearAccScale=<0.6 0.6 0.6>
    componentLinearVelScale=<0.7 0.7 0.7>
    gravityOverride {z=-500}
    physicsBodyDefinitions [{
      boundBone {boneName=towel_l}
      boxExtents=<5 11 8>
      constraintSetup {angularLimitsMax=<15 5 5>, angularLimitsMin=<-15 -30 -25>, angularTarget=<0 0 0>}
      localJointOffset=<1 0 0>
     }]
   }
   n19 = Anim Dynamics - Bone: towel_r!(ComponentPose=@n18, bAlphaBoolEnabled:bool=@n17, ExternalForce="0,0,0") -> (Pose) {
    alphaInputType=Bool
    angularDampingOverride=0.999
    angularSpringConstant=200
    bOverrideAngularDamping=true
    bUseGravityOverride=true
    simulationSpace=RootRelative
    alphaBoolBlend {blendInTime=0.1, blendOption=Linear, blendOutTime=0.1}
    boundBone {boneName=towel_r}
    componentAppliedLinearAccClamp=<200 200 700>
    componentLinearAccScale=<0.6 0.6 0.6>
    componentLinearVelScale=<0.7 0.7 0.7>
    gravityOverride {z=-500}
    physicsBodyDefinitions [{
      boundBone {boneName=towel_r}
      boxExtents=<8 11 11>
      constraintSetup {angularLimitsMax=<30 5 25>, angularLimitsMin=<-5 -30 -5>, angularTarget=<0 0 0>}
      localJointOffset=<1 0 0>
     }]
   }
   n20 = Anim Dynamics - Bone: arm_l!(ComponentPose=@n19, bAlphaBoolEnabled:bool=@n17, ExternalForce="0,0,0") -> (Pose) {
    alphaInputType=Bool
    angularDampingOverride=0.999
    angularSpringConstant=5000
    bAngularSpring=true
    bOverrideAngularDamping=true
    bUseGravityOverride=true
    simulationSpace=RootRelative
    alphaBoolBlend {blendInTime=0.1, blendOption=Linear, blendOutTime=0.1}
    boundBone {boneName=arm_l}
    componentAppliedLinearAccClamp=<190 190 690>
    componentLinearAccScale=<0.65 0.65 0.65>
    componentLinearVelScale=<0.75 0.75 0.75>
    gravityOverride {z=-500}
    physicsBodyDefinitions [{
      boundBone {boneName=arm_l}
      boxExtents=<9 11 8>
      constraintSetup {angularLimitsMax=<5 30 5>, angularLimitsMin=<-5 -25 -50>, angularTarget=<0 0 0>}
      localJointOffset=<1 0 0>
     }]
   }
   n21 = Anim Dynamics - Bone: arm_r!(ComponentPose=@n20, bAlphaBoolEnabled:bool=@n17, ExternalForce="0,0,0") -> (Pose) {
    alphaInputType=Bool
    angularDampingOverride=0.999
    angularSpringConstant=5000
    bAngularSpring=true
    bOverrideAngularDamping=true
    bUseGravityOverride=true
    simulationSpace=RootRelative
    alphaBoolBlend {blendInTime=0.1, blendOption=Linear, blendOutTime=0.1}
    boundBone {boneName=arm_r}
    componentAppliedLinearAccClamp=<190 190 690>
    componentLinearAccScale=<0.65 0.65 0.65>
    componentLinearVelScale=<0.75 0.75 0.75>
    gravityOverride {z=-500}
    physicsBodyDefinitions [{
      boundBone {boneName=arm_r}
      boxExtents=<9 11 8>
      constraintSetup {angularLimitsMax=<5 30 50>, angularLimitsMin=<-5 -25 -5>, angularTarget=<0 0 0>}
      localJointOffset=<1 0 0>
     }]
   }
   n22 = Get bIsSkydiveFloating()
   n23 = "Transform (Modify) Bone - Bone: towel_l"!(ComponentPose=@n21, bAlphaBoolEnabled:bool=@n22, Translation="0,0,0", Rotation="18.016,-25.297,23.700", Scale="1,1,1") -> (Pose) {
    alphaInputType=Bool
    rotationMode=BMM_Additive
    rotationSpace=BCS_BoneSpace
    alphaBoolBlend {blendInTime=0.1, blendOption=Linear, blendOutTime=0.5}
    boneToModify {boneName=towel_l}
    rotation=<rot 18.016 -25.297 23.7>
   }
   n24 = "Transform (Modify) Bone - Bone: towel_r"!(ComponentPose=@n23, bAlphaBoolEnabled:bool=@n22, Translation="0,0,0", Rotation="11.838,37.947,-8.599", Scale="1,1,1") -> (Pose) {
    alphaInputType=Bool
    rotationMode=BMM_Additive
    rotationSpace=BCS_BoneSpace
    alphaBoolBlend {blendInTime=0.1, blendOption=Linear, blendOutTime=0.5}
    boneToModify {boneName=towel_r}
    rotation=<rot 11.838 37.947 -8.599>
   }
   n25 = Anim Dynamics - Bone: towel_l(ComponentPose=@n24, bAlphaBoolEnabled=@n22, ExternalForce="0,0,0") {
    alphaInputType=Bool
    angularDampingOverride=0.999
    angularSpringConstant=200
    bOverrideAngularDamping=true
    bUseGravityOverride=true
    simulationSpace=RootRelative
    alphaBoolBlend {blendInTime=0.1, blendOption=Linear, blendOutTime=0.5}
    boundBone {boneName=towel_l}
    componentAppliedLinearAccClamp=<200 200 700>
    componentLinearAccScale=<0.6 0.6 0.6>
    componentLinearVelScale=<0.7 0.7 0.7>
    physicsBodyDefinitions [{
      boundBone {boneName=towel_l}
      boxExtents=<10 11 10>
      constraintSetup {angularLimitsMax=<0 -10 10>, angularLimitsMin=<0 -30 10>, angularTarget=<0 0 0>}
      localJointOffset=<1 0 0>
     }]
   }
   n26 = Anim Dynamics - Bone: towel_r(ComponentPose=@n25, bAlphaBoolEnabled=@n22, ExternalForce="0,0,0") {
    alphaInputType=Bool
    angularDampingOverride=0.999
    angularSpringConstant=200
    bOverrideAngularDamping=true
    bUseGravityOverride=true
    simulationSpace=RootRelative
    alphaBoolBlend {blendInTime=0.1, blendOption=Linear, blendOutTime=0.5}
    boundBone {boneName=towel_r}
    componentAppliedLinearAccClamp=<200 200 700>
    componentLinearAccScale=<0.6 0.6 0.6>
    componentLinearVelScale=<0.7 0.7 0.7>
    physicsBodyDefinitions [{
      boundBone {boneName=towel_r}
      boxExtents=<8 11 11>
      constraintSetup {angularLimitsMax=<0 -10 -10>, angularLimitsMin=<-15 -20 -10>, angularTarget=<0 0 0>}
      localJointOffset=<1 0 0>
     }]
   }
   n27 = Anim Dynamics - Bone: arm_l(ComponentPose=@n26, bAlphaBoolEnabled=@n22, ExternalForce="0,0,0") {
    alphaInputType=Bool
    angularDampingOverride=0.999
    angularSpringConstant=5000
    bAngularSpring=true
    bOverrideAngularDamping=true
    bUseGravityOverride=true
    simulationSpace=RootRelative
    alphaBoolBlend {blendInTime=0.1, blendOption=Linear, blendOutTime=0.5}
    boundBone {boneName=arm_l}
    componentAppliedLinearAccClamp=<200 200 500>
    componentLinearAccScale=<0.3 0.3 0.3>
    componentLinearVelScale=<0.5 0.5 0.5>
    physicsBodyDefinitions [{
      boundBone {boneName=arm_l}
      boxExtents=<10 10 10>
      constraintSetup {angularLimitsMax=<5 0 5>, angularLimitsMin=<-5 -40 -40>, angularTarget=<0 0 0>}
      localJointOffset=<1 0 0>
     }]
   }
   n28 = Anim Dynamics - Bone: arm_r(ComponentPose=@n27, bAlphaBoolEnabled=@n22, ExternalForce="0,0,0") {
    alphaInputType=Bool
    angularDampingOverride=0.999
    angularSpringConstant=5000
    bAngularSpring=true
    bOverrideAngularDamping=true
    bUseGravityOverride=true
    simulationSpace=RootRelative
    alphaBoolBlend {blendInTime=0.1, blendOption=Linear, blendOutTime=0.5}
    boundBone {boneName=arm_r}
    componentAppliedLinearAccClamp=<200 200 500>
    componentLinearAccScale=<0.3 0.3 0.3>
    componentLinearVelScale=<0.5 0.5 0.5>
    physicsBodyDefinitions [{
      boundBone {boneName=arm_r}
      boxExtents=<10 10 10>
      constraintSetup {angularLimitsMax=<5 0 30>, angularLimitsMin=<-5 -40 -5>, angularTarget=<0 0 0>}
      localJointOffset=<1 0 0>
     }]
   }
   Output Pose(Result=Blend Poses by bool(BlendPose_0=Component To Local(ComponentPose=@n14), BlendPose_1=Component To Local(ComponentPose=@n28), bActiveValue=Get bIsInFrontEnd()) {blendType=Linear}) {layerGroup=None}
   Save cached pose 'Base_pose'(Pose=Input Pose()) {cacheName=Base_pose}
  ]}
 EventGraph {nodes [
   Event BlueprintUpdateAnimation()
   > Set NotInSkydive(NotInSkydive:bool=NOT Boolean(A=Get bIsSkydiveFloating())) -> (Output_Get:bool)
   TryGetPawnOwner()
  ]}
]
variables [NotInSkydive {type=boolean}]
